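# Sound cue for the start of burrowed travel in sand tunnels, spatialized with the player sand tunnel attenuation.
# A mixer layers a random travel-start sound with quieter layers: a random slow burrowed loop, a pitched-up random "snared end low" sound and a mono burrowed loop.
# Identical envelopes swell and fade both loop layers, making them transient tails.
# Modulators vary the slow loop, the snared layer and the whole mix.
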SoundCue /FNGameplayCues/Athena/Sounds/SandTunnels/SandTunnel_Burrowed_Travel_Start_Cue
root=Modulator_0
nodes [
 {
  id=Enveloper_0
  children [Modulator_1]
  properties {
   pitchCurve {editorCurveData {keys [{time=0}]}}
   volumeCurve {editorCurveData {keys [<at 0.0013 0.0204>,<at 0.5099 0.9712>,<at 1.3865 0.3797>,<at 3.3412 -0.031>]}}
  }
 }
 {
  id=Enveloper_1
  children [WavePlayer_3]
  properties {
   pitchCurve {editorCurveData {keys [{time=0}]}}
   volumeCurve {editorCurveData {keys [<at 0.0013 0.0204>,<at 0.5099 0.9712>,<at 1.3865 0.3797>,<at 3.3412 -0.031>]}}
  }
 }
 {id=Mixer_0, children [Random_2,Enveloper_0,Modulator_2,Enveloper_1], properties {inputVolume [1,0.5,0.3,0.5]}}
 {id=Modulator_0, children [Mixer_0], properties {pitchMin=0.9}}
 {id=Modulator_1, children [Random_3], properties {pitchMax=1.15, pitchMin=1}}
 {id=Modulator_2, children [Random_1], properties {pitchMax=1.4, pitchMin=1.3}}
 {id=Random_1, children [WavePlayer_0,WavePlayer_1,WavePlayer_2], properties {weights [1,1,1]}}
 {id=Random_2, children [WavePlayer_6,WavePlayer_7,WavePlayer_8], properties {weights [1,1,1]}}
 {id=Random_3, children [WavePlayer_9, WavePlayer_10], properties {weights [1,1]}}
 {id=WavePlayer_0, properties {soundWaveAssetPtr=&/SandTunnel_Snared_End_Low_01}}
 {id=WavePlayer_1, properties {soundWaveAssetPtr=&/SandTunnel_Snared_End_Low_02}}
 {id=WavePlayer_10, properties {soundWaveAssetPtr=&/SandTunnel_Burrowed_Loop_Slow_02}}
 {id=WavePlayer_2, properties {soundWaveAssetPtr=&/SandTunnel_Snared_End_Low_03}}
 {id=WavePlayer_3, properties {soundWaveAssetPtr=&/SandTunnel_Burrowed_Loop_Mono_01}}
 {id=WavePlayer_6, properties {soundWaveAssetPtr=&/SandTunnel_Burrowed_Travel_Start_01}}
 {id=WavePlayer_7, properties {soundWaveAssetPtr=&/SandTunnel_Burrowed_Travel_Start_02}}
 {id=WavePlayer_8, properties {soundWaveAssetPtr=&/SandTunnel_Burrowed_Travel_Start_03}}
 {id=WavePlayer_9, properties {soundWaveAssetPtr=&/SandTunnel_Burrowed_Loop_Slow_01}}
]
properties {
 attenuationSettings=/FNGameplayCues/Sounds/Attenuation/Fort_Player_SandTunnel_Att
 maxDistance=7000
 soundClassObject=&Fort_Gadgets_Items
 volumeMultiplier=0.2
}
