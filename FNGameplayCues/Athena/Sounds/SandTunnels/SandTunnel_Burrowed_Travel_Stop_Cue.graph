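# Sound cue for the end of burrowed sand tunnel travel, judging by its name, using player sand tunnel attenuation and the gadgets/items sound class.
# A mixer layers one of three burrowed travel start waves at full volume with a quieter random pick from three low snared-end waves.
# Each layer passes through its own modulator for variation, with a lowered minimum pitch on the travel layer.
# The stop sound reuses travel start samples.
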SoundCue /FNGameplayCues/Athena/Sounds/SandTunnels/SandTunnel_Burrowed_Travel_Stop_Cue
root=Mixer_0
nodes [
 {id=Mixer_0, children [Modulator_0,Modulator_2], properties {inputVolume [1,0.3]}}
 {id=Modulator_0, children [Random_2], properties {pitchMin=0.9}}
 {id=Modulator_2, children [Random_1]}
 {id=Random_1, children [WavePlayer_0,WavePlayer_1,WavePlayer_2], properties {weights [1,1,1]}}
 {id=Random_2, children [WavePlayer_6,WavePlayer_7,WavePlayer_8], properties {weights [1,1,1]}}
 {id=WavePlayer_0, properties {soundWaveAssetPtr=&/SandTunnel_Snared_End_Low_01}}
 {id=WavePlayer_1, properties {soundWaveAssetPtr=&/SandTunnel_Snared_End_Low_02}}
 {id=WavePlayer_2, properties {soundWaveAssetPtr=&/SandTunnel_Snared_End_Low_03}}
 {id=WavePlayer_6, properties {soundWaveAssetPtr=&/SandTunnel_Burrowed_Travel_Start_01}}
 {id=WavePlayer_7, properties {soundWaveAssetPtr=&/SandTunnel_Burrowed_Travel_Start_02}}
 {id=WavePlayer_8, properties {soundWaveAssetPtr=&/SandTunnel_Burrowed_Travel_Start_03}}
]
properties {
 attenuationSettings=/FNGameplayCues/Sounds/Attenuation/Fort_Player_SandTunnel_Att
 maxDistance=7000
 soundClassObject=&Fort_Gadgets_Items
 volumeMultiplier=0.175
}
