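# Sound cue for emerging from a snared sand tunnel, attenuated in space and routed to the gadgets/items sound class.
# A mixer layers randomized snare-end, low snare-end, burrow-emerge and explosion-impact variations with a borrowed Storm King weak-point destruction sound.
# Modulators pitch both snare-end layers up, and an envelope fades the Storm King layer out over about two seconds.
# The whole mix runs through a very short fade-in envelope and a root modulator.
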SoundCue /FNGameplayCues/Athena/Sounds/SandTunnels/SandTunnel_Snared_Emerge_Cue
root=Modulator_0
nodes [
 {
  id=Enveloper_0
  children [Mixer_0]
  properties {pitchCurve {editorCurveData {keys [{time=0}]}}, volumeCurve {editorCurveData {keys [<at 0.0029 0.0182>,<at 0.071 1>]}}}
 }
 {
  id=Enveloper_1
  children [WavePlayer_12]
  properties {
   pitchCurve {editorCurveData {keys [{time=0}]}}
   volumeCurve {editorCurveData {keys [
      <at 0 1>
      {arriveTangent=-0.3333, interpMode=RCIM_Cubic, leaveTangent=-1.1858, tangentMode=RCTM_Break, time=0.3964, value=0.9978}
      <at 2.064 0.0049>
     ]}}
  }
 }
 {id=Mixer_0, children [Modulator_1,Random_0,Modulator_2,Modulator_3,Enveloper_1], properties {inputVolume [1.2,2,0.8,1.2,1]}}
 {id=Modulator_0, children [Enveloper_0], properties {volumeMax=1, volumeMin=1}}
 {id=Modulator_1, children [Random_2], properties {pitchMax=1.2, pitchMin=1.2, volumeMax=1, volumeMin=1}}
 {id=Modulator_2, children [Random_1], properties {pitchMax=1.4, pitchMin=1.3}}
 {id=Modulator_3, children [Random_3], properties {volumeMax=1, volumeMin=1}}
 {id=Random_0, children [WavePlayer_2,WavePlayer_3], properties {weights [1,1]}}
 {id=Random_1, children [WavePlayer_0,WavePlayer_7,WavePlayer_8], properties {weights [1,1,1]}}
 {id=Random_2, children [WavePlayer_1,WavePlayer_6,WavePlayer_5], properties {weights [1,1,1]}}
 {id=Random_3, children [WavePlayer_9, WavePlayer_10, WavePlayer_11], properties {weights [1,1,1]}}
 {id=WavePlayer_0, properties {soundWaveAssetPtr=&/SandTunnel_Snared_End_Low_01}}
 {id=WavePlayer_1, properties {soundWaveAssetPtr=&/SandTunnel_Snared_End_01}}
 {id=WavePlayer_10, properties {soundWaveAssetPtr=/Game/Sounds/Fort_Impact_Sounds/Demolition/Defaults/Explo_Impact_Layer02}}
 {id=WavePlayer_11, properties {soundWaveAssetPtr=/Game/Sounds/Fort_Impact_Sounds/Demolition/Defaults/Explo_Impact_Layer03}}
 {id=WavePlayer_12, properties {soundWaveAssetPtr=/Game/Sounds/Fort_Monster_Storm_King/SFX/Final_SFX/StormKing_WeakPoint_Destroyed_v2_02}}
 {id=WavePlayer_2, properties {soundWaveAssetPtr=&/SandTunnel_Burrowed_Emerge_01}}
 {id=WavePlayer_3, properties {soundWaveAssetPtr=&/SandTunnel_Burrowed_Emerge_02}}
 {id=WavePlayer_5, properties {soundWaveAssetPtr=&/SandTunnel_Snared_End_03}}
 {id=WavePlayer_6, properties {soundWaveAssetPtr=&/SandTunnel_Snared_End_02}}
 {id=WavePlayer_7, properties {soundWaveAssetPtr=&/SandTunnel_Snared_End_Low_02}}
 {id=WavePlayer_8, properties {soundWaveAssetPtr=&/SandTunnel_Snared_End_Low_03}}
 {id=WavePlayer_9, properties {soundWaveAssetPtr=/Game/Sounds/Fort_Impact_Sounds/Demolition/Defaults/Explo_Impact_Layer01}}
]
properties {
 attenuationSettings=/FNGameplayCues/Sounds/Attenuation/Fort_Player_SandTunnel_Snared_Att
 maxDistance=5000
 soundClassObject=&Fort_Gadgets_Items
 volumeMultiplier=0.35
}
