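# Looping sound cue for the snared state in Sand Tunnels, routed to the gadgets and items sound class with player-specific snared attenuation.
# A Mixer blends a main loop with a slightly boosted layer loop.
# An Enveloper ramps the mix up over roughly half a second, then a Modulator applies a fixed downward pitch shift without randomization.
# The cue plays at a reduced overall volume.
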SoundCue /FNGameplayCues/Athena/Sounds/SandTunnels/SandTunnel_Snared_Loop_Cue
root=Modulator_0
nodes [
 {
  id=Enveloper_0
  children [Mixer_0]
  properties {pitchCurve {editorCurveData {keys [{time=0}]}}, volumeCurve {editorCurveData {keys [<at 0.0462 0.0138>,<at 0.4962 1.0007>]}}}
 }
 {id=Mixer_0, children [WavePlayer_1,WavePlayer_2], properties {inputVolume [1,1.15]}}
 {id=Modulator_0, children [Enveloper_0], properties {pitchMax=0.8, pitchMin=0.8, volumeMax=1, volumeMin=1}}
 {id=WavePlayer_1, properties {bLooping=true, soundWaveAssetPtr=&/SandTunnel_Snared_Loop_01}}
 {id=WavePlayer_2, properties {bLooping=true, soundWaveAssetPtr=&/SandTunnel_Snared_Layer_Loop_01}}
]
properties {
 attenuationSettings=/FNGameplayCues/Sounds/Attenuation/Fort_Player_SandTunnel_Snared_Att
 maxDistance=5000
 soundClassObject=&Fort_Gadgets_Items
 volumeMultiplier=0.2
}
