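# Sound cue for the start of being snared by a sand tunnel, part of the Athena gameplay cue sounds.
# A Random node picks one of three equally weighted start waves, and a Modulator varies its pitch slightly on each play.
# It plays quietly through the gadgets and items sound class, spatialized by the sand tunnel snared attenuation settings.
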SoundCue /FNGameplayCues/Athena/Sounds/SandTunnels/SandTunnel_Snared_Start_Cue
root=Modulator_0
nodes [
 {id=Modulator_0, children [Random_0], properties {pitchMin=0.9}}
 {id=Random_0, children [WavePlayer_0,WavePlayer_2,WavePlayer_3], properties {weights [1,1,1]}}
 {id=WavePlayer_0, properties {soundWaveAssetPtr=&/SandTunnel_Snared_Start_01}}
 {id=WavePlayer_2, properties {soundWaveAssetPtr=&/SandTunnel_Snared_Start_02}}
 {id=WavePlayer_3, properties {soundWaveAssetPtr=&/SandTunnel_Snared_Start_03}}
]
properties {
 attenuationSettings=/FNGameplayCues/Sounds/Attenuation/Fort_Player_SandTunnel_Snared_Att
 maxDistance=5000
 soundClassObject=&Fort_Gadgets_Items
 volumeMultiplier=0.2
}
