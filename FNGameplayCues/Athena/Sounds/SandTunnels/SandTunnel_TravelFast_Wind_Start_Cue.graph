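# Sound cue for the wind burst at the start of fast sand tunnel travel, reusing a parachute skydive loop-enter wave.
# The wave feeds a Mixer with a slight boost and an empty second input, then a Modulator that randomly lowers pitch for variation.
# It plays in the gadgets and items sound class at reduced volume, using sand tunnel player attenuation with an attenuation override enabled.
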SoundCue /FNGameplayCues/Athena/Sounds/SandTunnels/SandTunnel_TravelFast_Wind_Start_Cue
root=Modulator_0
nodes [
 {id=Mixer_0, children [WavePlayer_0,null], properties {inputVolume [1.2,1]}}
 {id=Modulator_0, children [Mixer_0], properties {pitchMax=1, pitchMin=0.8, volumeMax=1, volumeMin=1}}
 {id=WavePlayer_0, properties {soundWaveAssetPtr=/Game/Athena/Sounds/Wind_Falling/sfx_parachute_skydive_loop_enter_01}}
]
properties {
 attenuationSettings=/FNGameplayCues/Sounds/Attenuation/Fort_Player_SandTunnel_Att
 bOverrideAttenuation=true
 maxDistance=4000
 soundClassObject=&Fort_Gadgets_Items
 volumeMultiplier=0.6
 attenuationOverrides {stereoSpread=3000}
}
