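# First-person looping Sound Cue for burrowed sand tunnel travel, using a dedicated 1P attenuation and the gadgets/items sound class.
# A crossfade on ShowTrailFX moves from the slow travel loop to the fast one, with Velocity modulating slow-loop volume and pitch and ShowTrailFX doing the same for fast.
# A boosted skydive wind loop with random pitch is mixed on top, its volume driven by ShowSpeedLines.
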SoundCue /FNGameplayCues/Athena/Sounds/SandTunnels/SandTunnel_Travel_Loop_1p_Cue
root=Mixer_3
nodes [
 {id=Mixer_0, children [Mixer_2], properties {inputVolume [2]}}
 {id=Mixer_1, children [WavePlayer_8], properties {inputVolume [0.75]}}
 {id=Mixer_2, children [WavePlayer_9], properties {inputVolume [1.45]}}
 {id=Mixer_3, children [ParamCrossFade_0, ModulatorContinuous_2], properties {inputVolume [1,3]}}
 {
  id=ModulatorContinuous_0
  children [Mixer_1]
  properties {pitchModulationParams {default=0, minOutput=0.8, parameterName=Velocity}, volumeModulationParams {default=0, parameterName=Velocity}}
 }
 {
  id=ModulatorContinuous_1
  children [Mixer_0]
  properties {pitchModulationParams {default=0, minOutput=0.8, parameterName=ShowTrailFX}, volumeModulationParams {default=0, parameterName=ShowTrailFX}}
 }
 {
  id=ModulatorContinuous_2
  children [Modulator_4]
  properties {
   pitchModulationParams {default=0, minOutput=1, parameterName=ShowSpeedLines}
   volumeModulationParams {default=0, parameterName=ShowSpeedLines}
  }
 }
 {id=Modulator_4, children [WavePlayer_14], properties {pitchMax=1, pitchMin=0.8, volumeMax=1, volumeMin=1}}
 {
  id=ParamCrossFade_0
  children [ModulatorContinuous_0, ModulatorContinuous_1]
  properties {
   paramName=ShowTrailFX
   crossFadeInput [
    {fadeInDistanceEnd=0, fadeInDistanceStart=0, fadeOutDistanceEnd=1, fadeOutDistanceStart=0, volume=1}
    {fadeInDistanceEnd=1, fadeInDistanceStart=0, fadeOutDistanceEnd=100, fadeOutDistanceStart=100, volume=1}
   ]
  }
 }
 {id=WavePlayer_14, properties {bLooping=true, soundWaveAssetPtr=/Game/Athena/Sounds/Wind_Falling/AMB_Wind_Skydive_Horizontal_Loop_01}}
 {
  id=WavePlayer_8
  properties {bLooping=true, soundWaveAssetPtr=&/SandTunnel_Burrowed_Travel_Loop_Slow_01}
 }
 {
  id=WavePlayer_9
  properties {bLooping=true, soundWaveAssetPtr=&/SandTunnel_Burrowed_Travel_Loop_Fast_01}
 }
]
properties {
 attenuationSettings=/FNGameplayCues/Sounds/Attenuation/Fort_Player_SandTunnel_1p_Att
 maxDistance=8000
 soundClassObject=&Fort_Gadgets_Items
 volumeMultiplier=0.25
}
